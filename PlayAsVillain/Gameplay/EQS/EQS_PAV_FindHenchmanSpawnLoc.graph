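# Environment query for Play As Villain that finds a spawn location for a henchman in a donut around the querier.
# A spiral donut generator creates candidate points, which are tested for pathfinding with a water-avoiding nav filter, navmesh projection, a dot-product threshold and world-static overlap.
# Scoring combines an obstruction test against the querier, a preference for nearer points and a random factor for variety.
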EnvQuery /PlayAsVillain/Gameplay/EQS/EQS_PAV_FindHenchmanSpawnLoc
options [{
  generator {
   name=EnvQueryGenerator_Donut_2
   class=EnvQueryGenerator_Donut
   properties {
    bUseSpiralPattern=true
    center=&EnvQueryContext_Querier
    innerRadius=600
    numberOfRings=3
    outerRadius=1000
    pointsPerRing=20
    projectionData {extentX=0, navigationFilter=null, postProjectionVerticalOffset=0, traceMode=None}
   }
  }
  tests [
   EnvQueryTest_Pathfinding_0 {
    class=EnvQueryTest_Pathfinding
    properties {
     filterClass=&NavFilter_Phoebe_AvoidWater_C
     pathFromContext {dataBinding=null, dataField=BoolValue, defaultValue=true}
    }
   }
   EnvQueryTest_Project_4 {
    class=EnvQueryTest_Project
    properties {
     scoringFactor=0
     testComment=""
     testPurpose=Filter
     projectionData {navigationFilter=null, postProjectionVerticalOffset=50, projectDown=500, projectUp=0, serializedChannel=ECC_Visibility, traceMode=Navigation}
    }
   }
   EnvQueryTest_Dot_0 {
    class=EnvQueryTest_Dot
    properties {
     clampMaxType=None
     clampMinType=SpecifiedValue
     floatValueMax=1
     floatValueMin=0.6
     scoreClampMax=0
     scoreClampMin=0
     scoringFactor=1
     testComment=""
     testPurpose=Filter
    }
   }
   EnvQueryTest_Overlap_1 {
    class=EnvQueryTest_Overlap
    properties {overlapData {extentX=50, extentY=50, extentZ=50, overlapChannel=ECC_WorldStatic, shapeOffset=<0 0 50>}}
   }
   FortQueryTest_IsObstructed_0 {class=FortQueryTest_IsObstructed, properties {context=&EnvQueryContext_Querier, scoringFactor=2, testPurpose=Score}}
   EnvQueryTest_Distance_0 {
    class=EnvQueryTest_Distance
    properties {
     distanceTo=&EnvQueryContext_Querier
     filterType=Range
     floatValueMin=0
     scoringEquation=InverseLinear
     scoringFactor=1
     testComment=""
     testMode=Distance3D
     testPurpose=Score
    }
   }
   FortQueryTest_Random_0 {class=FortQueryTest_Random, properties {floatValueMax=0, scoringFactor=1}}
  ]
 }]
